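# Environment query that generates candidate spawn points for the PlotTwist Force Projectile power around the querier.
# A donut generator lays out evenly ringed, non-spiral points in a band away from the querier, projected by geometry channel trace.
# A single Fortnite random test weights the candidates with square-root scoring, so selection does not depend on distance, visibility or pathing tests.
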EnvQuery /PlotTwist/ForcePowers/ForceProjectile/EQS_PlotTwist_Force_Projectile_SpawnPoints
options [{
  generator {
   name=EnvQueryGenerator_Donut_0
   class=EnvQueryGenerator_Donut
   properties {
    bUseSpiralPattern=false
    center=&EnvQueryContext_Querier
    innerRadius=400
    numberOfRings=4
    outerRadius=700
    pointsPerRing=16
    projectionData {extentX=0, navigationFilter=null, postProjectionVerticalOffset=0, traceMode=GeometryByChannel}
   }
  }
  tests [FortQueryTest_Random_0 {class=FortQueryTest_Random, properties {floatValueMax=0, scoringEquation=SquareRoot, scoringFactor=2}}]
 }]
