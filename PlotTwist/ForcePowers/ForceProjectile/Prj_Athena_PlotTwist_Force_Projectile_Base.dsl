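; Base projectile Blueprint for the PlotTwist Force power: a randomly meshed rock that lifts from the ground, hovers, then is thrown and explodes on impact.
; Begin play line-traces the ground for water or ground spawn FX, cracked decals and physics-field forces, while a timeline plays the lift.
; Throwing arms the explosion, restores gravity and raises replicated max speed.
; Hits play water or impact GameplayCues; cosmetics are gated, with some reduced on low-end platforms.
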(event Custom|SetGravity
  ;[1 1 1] Set Gravity
  (Variables|/Script/Engine.ProjectileMovementComponent|SetProjectileGravityScale (Variables|Default|GetProjectileFinalGravityScale) (Variables|Movement|GetProjectileMovementComponent)))

(event EventHandleThrown
  (CallFunction|Thrown))

(event Custom|Thrown
  ;[1 1 1] Event Thrown
  (Utilities|FlowControl|DoOnce false
    (:Completed
      (Utilities|FlowControl|Sequence
        (:then_0
          (Variables|UtilityVars|SetHasBeenThrown true)
          (Variables|UtilityVars|SetCanExplode true)
          (CallFunction|SetGravity)
          (Collision|SetCollisionEnabled (Variables|Collision|GetCapsuleComponent) "QueryOnly")
          (Variables|Default|SetRepMaxSpeed 1e+05))
        (:then_1
          (CallFunction|CEThrowOutFX))))))

(event Custom|CE_ThrowOutFX
  (bind _ns_forcepower_projectile (Variables|Default|GetNS_ForcePower_Projectile))
  ;[1 1 1] Throw Out FX
  (Audio|Components|Audio|Play (Variables|Default|GetAudio%20Thrown) 0.0) ; Thrown Loop
  (Audio|PlaySoundatLocation (Variables|Sound|GetSound_Thrown) (Transformation|GetActorLocation)) ; Thrown Burst
  (Audio|Components|Audio|FadeOut (Variables|Default|GetAudio%20Hover%20) 0.25 0.0 "Linear" :!class /Script/Engine.AudioComponent)
  (Networking|CanExecuteCosmeticEvents
    (:True
      (Components|Activation|Activate _ns_forcepower_projectile true)
      (Effects|Components|ParticleSystem|SetBoolParameter _ns_forcepower_projectile "StartThrowOut" true))))

(event Projectile|EventSetHomingData
  (params (SelfActor :type Actor%20Object%20Reference) (NewHomingData :type Projectile%20Homing%20Data%20Structure%20%28by%20ref%29))
  ;[1 1 1] Event Set Homing Data
  (|CallParentFunction SelfActor NewHomingData)
  (Variables|UtilityVars|SetCanExplode true))

(event EventBeginPlay
  ;[1 1 1] Event Begin Play
  (|CallParentFunction)
  (CallFunction|PlayRockSpawnFX)
  (CallFunction|SetProjectileInitialScaleandOffsetPosition)
  (CallFunction|SetRandomRotationVector)
  (CallFunction|ApplyNonRVT))

(event EventOnResumeSimulation
  ;[1 1 1] Event On Resume Simulation
  (|CallParentFunction))

(event Custom|ApplyNonRVT
  ; Non RVT for Creative / Playground Modes
  (if (Fortnite|IsCreativeorPlaygroundModePlaylist)
    (Networking|CanExecuteCosmeticEvents
      (:True
        (bind _result (Utilities|LoadFromSoftMaterial (Variables|Default|GetNonRVTRockMaterial))
          (:Completed
            (Rendering|Material|SetMaterial (Variables|Default|GetProjectileMesh) 0 _result)))))))

(event Custom|PlayRockSpawnFX
  (bind _returnvalue 0.5)
  (bind _returnvalue_2 0.2)
  (bind _returnvalue_3 0.66)
  (bind _returnvalue_4 3.5e+02)
  (bind _returnvalue_5 (Math|Random|RandomFloatinRange 0.0 0.35))
  ;[1 1 1] Rock Spawning FX
  (Networking|CanExecuteCosmeticEvents
    (:True
      (bind _outhit (Collision|LineTraceForObjects (+ (Transformation|GetActorLocation) "0.000000,0.000000,128.000000") (+ (Transformation|GetActorLocation) "0.000000,0.000000,-256.000000") (Variables|UtilityVars|GetTraceObjectTypes)))
      (Variables|UtilityVars|SetSpawnLocation (.Location (Collision|BreakHitResult _outhit)))
      (Variables|UtilityVars|SetPhysMatAtSpawnPoint (.PhysMat (Collision|BreakHitResult _outhit)))
      (if (Variables|ProjectileEffects|GetbNativeShouldPlaySpawnFX)
        (CallFunction|SpawnPhysicForce (Variables|UtilityVars|GetSpawnLocation) 4.5e+02 0.4)
        (ForceFeedback|SpawnForceFeedbackatLocation :ForceFeedbackEffect (Variables|ForceFeedback|GetFF_Spawn) :Location (Variables|UtilityVars|GetSpawnLocation) :bLooping false :IntensityMultiplier 1.0 :StartTime 0.0 :AttenuationSettings (Variables|ForceFeedback|GetFFAttenuation_Spawn) :bAutoDestroy true)
        (if (.bSuccess (Utilities|Casting|CastToFortWaterBodyBP (.HitActor (Collision|BreakHitResult _outhit))))
          (Utilities|FlowControl|Delay 0.28)
          (Utilities|AllowJanus
            (:Allow
              (bind _returnvalue_6 (Development|GetConsoleVariableIntValue "r.PhysicsField.EnableField"))
              (if (== _returnvalue_6 1)
                (Game|SpawnActorfromClass "/Game/Effects/Niagara/Blueprints/PhsyicsFields/FN_RadialForce.FN_RadialForce_C" (Transformation|GetWorldTransform (Variables|Collision|GetCapsuleComponent)) "Undefined" "MultiplyWithRoot" 0 _returnvalue_2 _returnvalue_3 _returnvalue_4 _returnvalue _returnvalue_5 0)
                (else
                  _)))
            (:Nope
              _))
          (bind _returnvalue_7 (Effects|Components|ParticleSystem|SpawnFXAtLocation (Variables|ProjectileEffects|GetFX_Spawn_Water) (Variables|UtilityVars|GetSpawnLocation))) ; SpwanWaterSplash
          (Effects|Components|ParticleSystem|SetFloatParameter _returnvalue_7 "EffectScale" 0.5)
          (Audio|PlaySoundatLocation (Variables|Sound|GetSound_Spawn_Water) (Variables|UtilityVars|GetSpawnLocation)) ; SFX Rocks Water Splash
          (else
            (Effects|Components|ParticleSystem|SpawnFXAtLocation (Variables|ProjectileEffects|GetFX_Spawn_Ground) (Variables|UtilityVars|GetSpawnLocation) (Math|Conversions|RotationFromXVector (.ImpactNormal (Collision|BreakHitResult _outhit))))
            (Utilities|AllowJanus
              (:Allow
                (bind _returnvalue_8 (Development|GetConsoleVariableIntValue "r.PhysicsField.EnableField"))
                (if (== _returnvalue_8 1)
                  (Game|SpawnActorfromClass "/Game/Effects/Niagara/Blueprints/PhsyicsFields/FN_RadialForce.FN_RadialForce_C" (Transformation|GetWorldTransform (Variables|Collision|GetCapsuleComponent)) "Undefined" "MultiplyWithRoot" 0 _returnvalue_2 _returnvalue_3 _returnvalue_4 _returnvalue _returnvalue_5 0)
                  (else
                    _)))
              (:Nope
                _))
            (Utilities|FlowControl|Delay 0.18)
            (CallFunction|CESpawnDecal (Variables|UtilityVars|GetSpawnLocation) (Variables|UtilityVars|GetPhysMatAtSpawnPoint))))))))

(event EventForcePlaySpawnFX
  ;[1 1 1] Rock Spawning FX
  (CallFunction|RepForcePlaySpawnFX))

(event Custom|RepForcePlaySpawnFX
  ;[1 1 1] Rock Spawning FX
  (Variables|ProjectileEffects|SetbNativeShouldPlaySpawnFX true)
  (CallFunction|PlayRockSpawnFX))

(event Custom|SetProjectileInitialScaleAndOffsetPosition
  ;[1 1 1] Rock lift up
  (Networking|CanExecuteCosmeticEvents
    (:True
      (Variables|Default|SetScale_Final (Math|Random|RandomFloatinRange (Variables|Default|GetScale_Final_Min) (Variables|Default|GetScale_Final_Max)))
      (Components|Timeline|PlayfromStart (Variables|Prj_Athena_PlotTwist_Force_Projectile_Base|GetTimeline)))))

(event Custom|RepSetShouldUseSpawnOffset
  (params (ShouldUseSpawnOffset :type Boolean))
  ;[1 1 1] Rock lift up
  (Variables|ProjectileEffects|SetbNativeShouldUseSpawnOffset ShouldUseSpawnOffset))

(event Collision|EventHit
  (params (MyComp :type Primitive%20Component%20Object%20Reference) (Other :type Actor%20Object%20Reference) (OtherComp :type Primitive%20Component%20Object%20Reference) (bSelfMoved :type Boolean) (HitLocation :type Vector :default "0, 0, 0") (HitNormal :type Vector :default "0, 0, 0") (NormalImpulse :type Vector :default "0, 0, 0") (Hit :type Hit%20Result%20Structure%20%28by%20ref%29))
  (bind (_bblockinghit _binitialoverlap _time _distance _location _impactpoint _normal _impactnormal _physmat) (Collision|BreakHitResult Hit))
  ;[1 1 1] Projectile Hit and Explode Events
  (if (Variables|UtilityVars|GetCanExplode)
    (Utilities|Casting|CastToWaterBody Other
      (:then
        (CallFunction|CEPlayHitWaterFX HitLocation HitNormal _physmat))
      (:CastFailed
        (CallFunction|CEPlayImpactFX HitLocation HitNormal _physmat)))
    (Delivery|DoExplosionatLocation (Transformation|GetActorLocation))))

(event Custom|CE_PlayImpactFX
  (params (HitLocation :type Vector) (HitNormal :type Vector) (PhysicalMaterial :type Physical%20Material%20Object%20Reference))
  ;[1 1 1] Impact FX
  (Networking|CanExecuteCosmeticEvents
    (:True
      (GameplayCue|ExecuteGameplayCue%28Burst%29Local self (Variables|ProjectileEffects|GetTag_GC_ImpactFX) (Ability|GameplayCue|MakeGameplayCueParameters :NormalizedMagnitude 0.0 :RawMagnitude 0.0 :Location HitLocation :Normal HitNormal :PhysicalMaterial PhysicalMaterial :GameplayEffectLevel 1 :AbilityLevel 1 :bReplicateLocationWhenUsingMinimalRepProxy false)))))

(event EventOnExploded
  (params (HitActors :type Array%20of%20Actor%20Object%20References) (HitResults :type Array%20of%20Hit%20Result%20Structures))
  ;[1 1 1] Projectile Hit and Explode Events
  (|CallParentFunction HitActors HitResults)
  (Components|Movement|StopMovementImmediately (Variables|Movement|GetProjectileMovementComponent))
  (Delivery|Kill)
  (Audio|Components|Audio|FadeOut (Variables|Default|GetAudio%20Thrown) 0.0 0.0 "Linear" :!class /Script/Engine.AudioComponent) ; Stop loop on impact
)

(event Custom|CE_PlayHitWaterFX
  (params (HitLocation :type Vector) (HitNormal :type Vector) (PhysicalMaterial :type Physical%20Material%20Object%20Reference))
  ;[1 1 1] Hit Water FX
  (Networking|CanExecuteCosmeticEvents
    (:True
      (GameplayCue|ExecuteGameplayCue%28Burst%29Local self (Variables|ProjectileEffects|GetTag_GC_HitWaterFX) (Ability|GameplayCue|MakeGameplayCueParameters :NormalizedMagnitude 0.0 :RawMagnitude 0.0 :Location HitLocation :Normal (Math|Vector|ProjectVectorontoPlane (Math|Vector|Normalize (Variables|/Script/Engine.MovementComponent|GetVelocity (Variables|Movement|GetProjectileMovementComponent))) HitNormal) :PhysicalMaterial PhysicalMaterial :GameplayEffectLevel 1 :AbilityLevel 1 :bReplicateLocationWhenUsingMinimalRepProxy false))
      (Utilities|FlowControl|Delay 0.4)
      (Audio|PlaySoundatLocation (Variables|Sound|GetSound_Water_Impact) (Transformation|GetActorLocation)) ; Water Impact SFX
    )))

(event EventTick
  (params (DeltaSeconds :type Float%20%28single-precision%29))
  ;[1 1 1] Event Tick - Random Rotate the rock mesh
  (|CallParentFunction DeltaSeconds)
  (Networking|CanExecuteCosmeticEvents
    (:True
      (Transformation|AddRelativeRotation (Variables|Default|GetProjectileMesh) (Math|Rotator|MakeRotator (.x (Variables|Rotation|GetRandomVector)) (.y (Variables|Rotation|GetRandomVector)) (.z (Variables|Rotation|GetRandomVector))) false false))))

(event Custom|SetRandomRotationVector
  ;[1 1 1] Set Random Rotation Vector
  (Networking|CanExecuteCosmeticEvents
    (:True
      (Variables|Rotation|SetRandomVector (* (Math|Vector|MakeVector (Math|Random|RandomFloatinRange -1.0 1.0) (Math|Random|RandomFloatinRange -1.0 1.0) (Math|Random|RandomFloatinRange -1.0 1.0)) (Math|Conversions|ToVector(Float) (Variables|Rotation|GetRotationSpeedMultiplier)))))))

(event Custom|CE_SpawnDecal
  (params (SpawnLocation :type Vector) (PhysicMaterial :type Physical%20Material%20Object%20Reference))
  ;[1 1 1] SpawnDecal
  (if (Math|Boolean|ANDBoolean (Utilities|Array|ContainsItem (Variables|Decal|GetDecalRequiredPhysMat) PhysicMaterial) (and (not (Fortnite|IsMobilePlatform)) (not (Class|PlatformHelpersLibrary|IsPcPerfMode))) (Variables|Decal|GetShouldSpawnDecal))
    (bind _returnvalue (Rendering|Decal|SpawnDecalatLocation (Variables|Decal|GetDecalMaterial_Default) (Variables|Decal|GetDecal_Size) SpawnLocation (Math|Rotator|MakeRotator (Math|Float|Lerp 0.0 3.6e+02 (Math|Random|RandomFloat)) -9e+01)))
    (Variables|Default|SetSpawnedDecal _returnvalue)
    (Transformation|SetRelativeScale3D (Variables|Default|GetSpawnedDecal) (Variables|Decal|GetDecal_Scale))
    (if (== PhysicMaterial "/Game/FortressPhysicalMaterials/Asphalt.Asphalt")
      (Rendering|Components|Decal|SetDecalMaterial (Variables|Default|GetSpawnedDecal) (Variables|Decal|GetDecalMaterial_Asphalt)) ; MD_Decal_GroundCrack_Desatuation_Inst
      (else
        (Rendering|Components|Decal|SetDecalMaterial (Variables|Default|GetSpawnedDecal) (Variables|Decal|GetDecalMaterial_NonAsphalt)) ; MD_Decal_GroundCrack_Inst
      ))
    (Rendering|Components|Decal|SetFadeOut (Variables|Default|GetSpawnedDecal) (Variables|Decal|GetDecal_FadeOut_StartDelay) (Variables|Decal|GetDecal_FadeOut_Duration) true)
    (Utilities|FlowControl|Sequence
      (:then_0
        (Utilities|FlowControl|Delay 0.2)
        (Transformation|SetRelativeScale3D (Variables|Default|GetSpawnedDecal) "0.800000,0.800000,0.800000"))
      (:then_1
        (Utilities|FlowControl|Delay 0.3)
        (Transformation|SetRelativeScale3D (Variables|Default|GetSpawnedDecal) "1.000000,1.000000,1.000000")))))

(fn OnRep_RepMaxSpeed
  (Variables|/Script/Engine.ProjectileMovementComponent|SetMaxSpeed (Variables|Default|GetRepMaxSpeed) (Variables|Movement|GetProjectileMovementComponent)))

(fn SpawnPhysicForce
  (params (Location :type Vector) (ImpactRadius :type Float%20%28double-precision%29) (Time :type Float%20%28double-precision%29))
  (if (and (not (or (Class|PlatformHelpersLibrary|IsPcPerfMode) (Fortnite|IsMobilePlatform))) (> (Rendering|GetRenderingDetailMode) 1))
    (Game|SpawnActorfromClass "/Game/Effects/Niagara/Blueprints/PhsyicsFields/FN_RadialForce.FN_RadialForce_C" (Math|Transform|MakeTransform Location) "Undefined" "MultiplyWithRoot" 0 3.0 Time ImpactRadius 0.5 0.05 0)))

(fn ConstructionScript
  (bind _rockmeshinstances (Variables|Default|GetRockMeshInstances))
  (bind _returnvalue (Utilities|Array|LastIndex _rockmeshinstances))
  (bind _returnvalue_2 (Math|Random|RandomIntegerinRange 0 _returnvalue))
  (bind _rockmeshinstances_2 (Variables|Default|GetRockMeshInstances))
  (bind _output (Utilities|Array|Get(acopy) _rockmeshinstances_2 _returnvalue_2))
  (|CallParentFunction)
  (if (Utilities|Array|IsNotEmpty (Variables|Default|GetRockMeshInstances))
    (Components|StaticMesh|SetStaticMesh (Variables|Default|GetProjectileMesh) _output)
    (Niagara|SetNiagaraVariable(StaticMesh) (Variables|Default|GetNS_ForcePower_Projectile) "StaticMeshSample" _output)
    (Rendering|Material|SetScalarParameterValueonMaterials (Variables|Default|GetProjectileMesh) "UseFresnelEmissive" 0.0) ; TurnOff blue glow: Not Using for now due to IP feedback
  ))
