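; Projectile Blueprint for the wood variant of the Force power projectile in the PlotTwist content.
; The construction script passes the projectile mesh's relative scale to the wood formation Niagara effect as its MeshScale variable.
; On BeginPlay it waits briefly, then makes the mesh visible, presumably once the forming effect has played.
; The tick and overlap events add no logic of their own.
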(event EventBeginPlay
  (|CallParentFunction)
  (Utilities|FlowControl|Delay 1.0)
  (Rendering|SetVisibility (Variables|Default|GetProjectileMesh) true false))

(event Collision|EventActorBeginOverlap
  (params (OtherActor :type Actor%20Object%20Reference)))

(event EventTick
  (params (DeltaSeconds :type Float%20%28single-precision%29))
  (|CallParentFunction DeltaSeconds) ; This node is disabled and will not be called.\nDrag off pins to build functionality.
)

(fn ConstructionScript
  (|CallParentFunction)
  (Niagara|SetNiagaraVariable(Vector3) (Variables|Default|GetNS_ForcePower_Projectile_Wood_Form) "MeshScale" (Variables|/Script/Engine.SceneComponent|GetRelativeScale3D (Variables|Default|GetProjectileMesh))))
