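# Sound cue for the rock impact of a Force Projectile boulder in the PlotTwist force powers.
# A quality-level switch feeds the first two slots from a random pick of six impact waves and the last two from a three-wave subset (04–06).
# The pick passes through a neutral modulator, a third-person weapon-fire sound class override and a root modulator that varies pitch slightly upward.
# A boulder-impact attenuation asset spatializes it, and a concurrency override caps simultaneous instances.
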SoundCue /PlotTwist/ForcePowers/ForceProjectile/Sound/ForceProjectile_BoulderImpact_Cue
root=Modulator_7
nodes [
 {id=Modulator_5, children [QualityLevel_0], properties {pitchMax=1, pitchMin=1, volumeMax=1, volumeMin=1}}
 {id=Modulator_7, children [SoundClass_0], properties {pitchMax=1.1, pitchMin=1}}
 {id=QualityLevel_0, children [Random_0,Random_0,Random_2,Random_2]}
 {
  id=Random_0
  children [WavePlayer_1, WavePlayer_7, WavePlayer_8, WavePlayer_9, WavePlayer_10, WavePlayer_11]
  properties {weights [1,1,1,1,1,1]}
 }
 {id=Random_2, children [WavePlayer_9, WavePlayer_10, WavePlayer_11], properties {weights [1,1,1]}}
 {id=SoundClass_0, children [Modulator_5], properties {soundClassOverride=&Fort_WeaponRanged_Fire_3P_PSM}}
 {id=WavePlayer_1, properties {soundWaveAssetPtr=&/Force_Projectile_RockImpact_01}}
 {id=WavePlayer_10, properties {soundWaveAssetPtr=&/Force_Projectile_RockImpact_05}}
 {id=WavePlayer_11, properties {soundWaveAssetPtr=&/Force_Projectile_RockImpact_06}}
 {id=WavePlayer_7, properties {soundWaveAssetPtr=&/Force_Projectile_RockImpact_02}}
 {id=WavePlayer_8, properties {soundWaveAssetPtr=&/Force_Projectile_RockImpact_03}}
 {id=WavePlayer_9, properties {soundWaveAssetPtr=&/Force_Projectile_RockImpact_04}}
]
properties {
 attenuationSettings=/Game/Sounds/Attenuation/Fort_ForcePowers_Projectile_BoulderImpact_Att
 bOverrideConcurrency=true
 maxDistance=14000
 soundClassObject=&Fort_WeaponRanged_Fire_3P_PSM
 volumeMultiplier=1.25
 attenuationOverrides {falloffDistance=6000, attenuationShapeExtents {x=1000}}
 concurrencyOverrides {maxCount=4}
}
