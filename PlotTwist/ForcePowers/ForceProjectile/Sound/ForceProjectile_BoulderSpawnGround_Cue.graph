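# Sound cue for a boulder spawning from the ground in the Force Projectile power, with close and distant layers.
# Each layer randomly picks one of three waves through its own modulator, attenuation asset and weapon-fire sound class override.
# A mixer plays both layers together, boosting the close one, under a root pitch modulator.
# Concurrency is overridden to cap simultaneous instances at two.
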SoundCue /PlotTwist/ForcePowers/ForceProjectile/Sound/ForceProjectile_BoulderSpawnGround_Cue
root=Modulator_7
nodes [
 {
  id=Attenuation_1
  children [Modulator_1]
  properties {attenuationSettings=/PlotTwist/ForcePowers/ForceProjectile/Sound/Settings/ForceThrow_RockSpawn_Att}
 }
 {
  id=Attenuation_7
  children [Modulator_8]
  properties {attenuationSettings=/PlotTwist/ForcePowers/ForceProjectile/Sound/Settings/ForceThrow_RockSpawn_Distant_Att}
 }
 {id=Mixer_0, children [Mixer_2,SoundClass_2], properties {inputVolume [2,1]}}
 {id=Mixer_2, children [SoundClass_3], properties {inputVolume [1]}}
 {id=Modulator_1, children [Random_3], properties {pitchMax=1.15, pitchMin=0.9, volumeMax=1, volumeMin=1}}
 {id=Modulator_7, children [Mixer_0], properties {pitchMax=1.1, pitchMin=1, volumeMax=1, volumeMin=1}}
 {id=Modulator_8, children [Random_1], properties {volumeMax=1, volumeMin=1}}
 {id=Random_1, children [WavePlayer_12, WavePlayer_13, WavePlayer_14], properties {bShouldExcludeFromBranchCulling=true, weights [1,1,1]}}
 {id=Random_3, children [WavePlayer_3,WavePlayer_4,WavePlayer_5], properties {bShouldExcludeFromBranchCulling=true, weights [1,1,1]}}
 {
  id=SoundClass_2
  children [Attenuation_7]
  properties {soundClassOverride=&Fort_WeaponRanged_Fire_Distant}
 }
 {
  id=SoundClass_3
  children [Attenuation_1]
  properties {soundClassOverride=&Fort_WeaponRanged_Fire_3P_PSM}
 }
 {id=WavePlayer_12, properties {soundWaveAssetPtr=&/Force_Projectile_BoulderSpawn_Distant_01}}
 {id=WavePlayer_13, properties {soundWaveAssetPtr=&/Force_Projectile_BoulderSpawn_Distant_02}}
 {id=WavePlayer_14, properties {soundWaveAssetPtr=&/Force_Projectile_BoulderSpawn_Distant_03}}
 {id=WavePlayer_3, properties {soundWaveAssetPtr=&/Force_Projectile_BoulderSpawn_01}}
 {id=WavePlayer_4, properties {soundWaveAssetPtr=&/Force_Projectile_BoulderSpawn_02}}
 {id=WavePlayer_5, properties {soundWaveAssetPtr=&/Force_Projectile_BoulderSpawn_03}}
]
properties {
 bOverrideConcurrency=true
 maxDistance=27000
 soundClassObject=&Fort_WeaponRanged_Fire_3P_PSM
 volumeMultiplier=1.12
 attenuationOverrides {falloffDistance=4500, attenuationShapeExtents {x=500}}
 concurrencyOverrides {maxCount=2}
}
