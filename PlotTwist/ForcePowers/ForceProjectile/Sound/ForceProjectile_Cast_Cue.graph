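# Sound cue for casting the force projectile power, blending a close cast layer with a quieter-mixed distant layer.
# An IsLocal branch plays close variants through louder 2D attenuation for the local case, and through close Throw attenuation when false or unset.
# The distant layer always plays alongside, with its own attenuation and distant-fire sound class.
# Each layer picks randomly from three waves, close ones get slight pitch variation, and concurrency is capped.
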SoundCue /PlotTwist/ForcePowers/ForceProjectile/Sound/ForceProjectile_Cast_Cue
root=Modulator_7
nodes [
 {
  id=Attenuation_0
  children [Modulator_0]
  properties {attenuationSettings=/PlotTwist/ForcePowers/ForceProjectile/Sound/Settings/ForceThrow_Cast_Throw_Close_Att}
 }
 {
  id=Attenuation_2
  children [Modulator_5]
  properties {attenuationSettings=/PlotTwist/ForcePowers/ForceProjectile/Sound/Settings/FortThrow_Cast_2D_Att}
 }
 {
  id=Attenuation_7
  children [Modulator_8]
  properties {attenuationSettings=/PlotTwist/ForcePowers/ForceProjectile/Sound/Settings/ForceThrow_Cast_Throw_Distant_Att}
 }
 {id=Branch_0, children [Mixer_1,Mixer_2,Mixer_2], properties {boolParameterName=IsLocal}}
 {id=Mixer_0, children [Branch_0,SoundClass_2], properties {inputVolume [1,0.6]}}
 {id=Mixer_1, children [SoundClass_0], properties {inputVolume [4]}}
 {id=Mixer_2, children [SoundClass_1], properties {inputVolume [0.7]}}
 {id=Modulator_0, children [Random_2], properties {pitchMax=1.1, pitchMin=1, volumeMax=1, volumeMin=1}}
 {id=Modulator_5, children [Random_0], properties {pitchMax=1.1, pitchMin=1, volumeMax=1, volumeMin=1}}
 {id=Modulator_7, children [Mixer_0], properties {pitchMax=1, pitchMin=1, volumeMax=1, volumeMin=1}}
 {id=Modulator_8, children [Random_1], properties {volumeMax=2, volumeMin=2}}
 {id=Random_0, children [WavePlayer_3,WavePlayer_4,WavePlayer_5], properties {weights [1,1,1]}}
 {id=Random_1, children [WavePlayer_6, WavePlayer_9, WavePlayer_10], properties {weights [1,1,1]}}
 {id=Random_2, children [WavePlayer_11, WavePlayer_12, WavePlayer_13], properties {weights [1,1,1]}}
 {
  id=SoundClass_0
  children [Attenuation_2]
  properties {soundClassOverride=&Fort_WeaponRanged_Fire_3P_PSM}
 }
 {
  id=SoundClass_1
  children [Attenuation_0]
  properties {soundClassOverride=&Fort_WeaponRanged_Fire_3P_PSM}
 }
 {
  id=SoundClass_2
  children [Attenuation_7]
  properties {soundClassOverride=&Fort_WeaponRanged_Fire_Distant}
 }
 {id=WavePlayer_10, properties {soundWaveAssetPtr=&/Force_Projectile_Cast_Distant_03}}
 {id=WavePlayer_11, properties {soundWaveAssetPtr=&/Force_Projectile_Cast_Close_01}}
 {id=WavePlayer_12, properties {soundWaveAssetPtr=&/Force_Projectile_Cast_Close_02}}
 {id=WavePlayer_13, properties {soundWaveAssetPtr=&/Force_Projectile_Cast_Close_03}}
 {id=WavePlayer_3, properties {soundWaveAssetPtr=&Force_Projectile_Cast_Close_01}}
 {id=WavePlayer_4, properties {soundWaveAssetPtr=&Force_Projectile_Cast_Close_02}}
 {id=WavePlayer_5, properties {soundWaveAssetPtr=&Force_Projectile_Cast_Close_03}}
 {id=WavePlayer_6, properties {soundWaveAssetPtr=&/Force_Projectile_Cast_Distant_01}}
 {id=WavePlayer_9, properties {soundWaveAssetPtr=&/Force_Projectile_Cast_Distant_02}}
]
properties {
 bOverrideConcurrency=true
 maxDistance=25000
 soundClassObject=&Fort_WeaponRanged_Fire_3P_PSM
 volumeMultiplier=0.2
 attenuationOverrides {falloffDistance=4500, attenuationShapeExtents {x=500}}
 concurrencyOverrides {maxCount=3}
}
