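# Sound cue for the Force Projectile boulder in flight, layering launch and in-air start one-shots with a looping air sound.
# Random, pitch-modulated launch and start variants share an in-air attenuation and mixer, which feeds the top mixer both directly and through a short random delay.
# The pitch-modulated loop runs through a volume and pitch envelope, smoothed Doppler and its own close attenuation.
# A concurrency override caps simultaneous instances.
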SoundCue /PlotTwist/ForcePowers/ForceProjectile/Sound/ForceProjectile_ProjectileBoulder_InAirLoop_Cue
root=Delay_1
nodes [
 {
  id=Attenuation_2
  children [Doppler_0]
  properties {attenuationSettings=/Game/Sounds/Attenuation/Fort_ForcePowers_Projectile_BoulderInAir_Close_Att}
 }
 {
  id=Attenuation_3
  children [Modulator_2]
  properties {attenuationSettings=/Game/Sounds/Attenuation/Fort_ForcePowers_Projectile_BoulderInAir_Att}
 }
 {
  id=Attenuation_4
  children [Modulator_0]
  properties {attenuationSettings=&Fort_ForcePowers_Projectile_BoulderInAir_Att}
 }
 {id=Delay_0, children [Mixer_1], properties {delayMax=0.6, delayMin=0.3}}
 {id=Delay_1, children [Mixer_0]}
 {id=Doppler_0, children [Enveloper_0], properties {bUseSmoothing=true, dopplerIntensity=2}}
 {
  id=Enveloper_0
  children [Modulator_1]
  properties {pitchCurve {editorCurveData {keys [<at 0 1.25>, {time=1}]}}, volumeCurve {editorCurveData {keys [<at 0 0 RCIM_Cubic>,<at 1 1>]}}}
 }
 {id=Mixer_0, children [Mixer_1, Delay_0, Attenuation_2], properties {inputVolume [1,1,1]}}
 {id=Mixer_1, children [Attenuation_3, Attenuation_4], properties {inputVolume [1,1]}}
 {id=Modulator_0, children [Random_2], properties {pitchMax=1.2, pitchMin=0.9, volumeMax=1, volumeMin=1}}
 {id=Modulator_1, children [WavePlayer_1], properties {pitchMax=1.2, pitchMin=0.85, volumeMax=1, volumeMin=1}}
 {id=Modulator_2, children [Random_3], properties {pitchMax=1.2, pitchMin=0.9, volumeMax=1, volumeMin=1}}
 {id=Random_2, children [WavePlayer_3,WavePlayer_4,WavePlayer_5], properties {bShouldExcludeFromBranchCulling=true, weights [1,1,1]}}
 {id=Random_3, children [WavePlayer_13, WavePlayer_14, WavePlayer_15], properties {bShouldExcludeFromBranchCulling=true, weights [1,1,1]}}
 {
  id=WavePlayer_1
  properties {bLooping=true, soundWaveAssetPtr=&/ForceProjectile_ProjectileBoulder_InAir_Loop}
 }
 {
  id=WavePlayer_13
  properties {soundWaveAssetPtr=&/ForceProjectile_ProjectileBoulder_InAir_Start_01}
 }
 {
  id=WavePlayer_14
  properties {soundWaveAssetPtr=&/ForceProjectile_ProjectileBoulder_InAir_Start_02}
 }
 {
  id=WavePlayer_15
  properties {soundWaveAssetPtr=&/ForceProjectile_ProjectileBoulder_InAir_Start_03}
 }
 {id=WavePlayer_3, properties {soundWaveAssetPtr=&/Force_Projectile_BoulderLaunch_01}}
 {id=WavePlayer_4, properties {soundWaveAssetPtr=&/Force_Projectile_BoulderLaunch_02}}
 {id=WavePlayer_5, properties {soundWaveAssetPtr=&/Force_Projectile_BoulderLaunch_03}}
]
properties {
 bOverrideConcurrency=true
 maxDistance=27000
 soundClassObject=&Fort_WeaponRanged_Fire_3P_PSM
 volumeMultiplier=0.8
 attenuationOverrides {falloffDistance=12000, attenuationShapeExtents {x=1000}}
 concurrencyOverrides {maxCount=3}
}
