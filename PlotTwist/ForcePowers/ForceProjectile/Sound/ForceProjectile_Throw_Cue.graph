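# Sound cue for the Force Projectile throw, blending a close layer and a distant layer.
# Each layer randomly picks one of three waves, passes through a modulator, and gets its own attenuation settings and ranged-weapon-fire sound class override.
# A mixer weights the close layer louder, then a root modulator adds pitch variation to the blend.
# The cue overrides concurrency to cap simultaneous instances.
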SoundCue /PlotTwist/ForcePowers/ForceProjectile/Sound/ForceProjectile_Throw_Cue
root=Modulator_7
nodes [
 {
  id=Attenuation_2
  children [Modulator_5]
  properties {attenuationSettings=/PlotTwist/ForcePowers/ForceProjectile/Sound/Settings/ForceThrow_Cast_Throw_Close_Att}
 }
 {
  id=Attenuation_7
  children [Modulator_8]
  properties {attenuationSettings=/PlotTwist/ForcePowers/ForceProjectile/Sound/Settings/ForceThrow_Cast_Throw_Distant_Att}
 }
 {id=Mixer_0, children [SoundClass_0,SoundClass_2], properties {inputVolume [0.9,0.5]}}
 {id=Modulator_5, children [Random_0], properties {pitchMax=1.1, pitchMin=1, volumeMax=1, volumeMin=1}}
 {id=Modulator_7, children [Mixer_0], properties {pitchMax=1.2, volumeMax=1, volumeMin=1}}
 {id=Modulator_8, children [Random_1], properties {volumeMax=1, volumeMin=1}}
 {id=Random_0, children [WavePlayer_20, WavePlayer_21, WavePlayer_22], properties {weights [1,1,1]}}
 {id=Random_1, children [WavePlayer_23, WavePlayer_24, WavePlayer_25], properties {weights [1,1,1]}}
 {
  id=SoundClass_0
  children [Attenuation_2]
  properties {soundClassOverride=&Fort_WeaponRanged_Fire_3P_PSM}
 }
 {
  id=SoundClass_2
  children [Attenuation_7]
  properties {soundClassOverride=&Fort_WeaponRanged_Fire_Distant}
 }
 {id=WavePlayer_20, properties {soundWaveAssetPtr=&/Force_Projectile_Throw_Close_01}}
 {id=WavePlayer_21, properties {soundWaveAssetPtr=&/Force_Projectile_Throw_Close_02}}
 {id=WavePlayer_22, properties {soundWaveAssetPtr=&/Force_Projectile_Throw_Close_03}}
 {id=WavePlayer_23, properties {soundWaveAssetPtr=&/Force_Projectile_Throw_Distant_01}}
 {id=WavePlayer_24, properties {soundWaveAssetPtr=&/Force_Projectile_Throw_Distant_02}}
 {id=WavePlayer_25, properties {soundWaveAssetPtr=&/Force_Projectile_Throw_Distant_03}}
]
properties {
 bOverrideConcurrency=true
 maxDistance=25000
 soundClassObject=&Fort_WeaponRanged_Fire_3P_PSM
 volumeMultiplier=0.5
 attenuationOverrides {falloffDistance=4500, attenuationShapeExtents {x=500}}
 concurrencyOverrides {maxCount=2, retriggerTime=1}
}
