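# Sound cue that plays the water splash sound for the Force Projectile power.
# A Random node picks one of five equally weighted splash waves and feeds it through a neutral Modulator, which adds no pitch or volume variation.
# The cue slightly lowers pitch and volume, uses a dedicated splash attenuation asset and a third-person ranged weapon fire sound class.
# Overridden concurrency caps simultaneous instances at four, stopping the oldest.
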SoundCue /PlotTwist/ForcePowers/ForceProjectile/Sound/Force_Projectile_WaterSplash_Cue
root=Modulator_2
nodes [
 {id=Modulator_2, children [Random_0], properties {pitchMax=1, pitchMin=1, volumeMax=1, volumeMin=1}}
 {id=Random_0, children [WavePlayer_0,WavePlayer_6,WavePlayer_7,WavePlayer_8,WavePlayer_9], properties {weights [1,1,1,1,1]}}
 {id=WavePlayer_0, properties {soundWaveAssetPtr=&/Force_Projectile_WaterSplash_01}}
 {id=WavePlayer_6, properties {soundWaveAssetPtr=&/Force_Projectile_WaterSplash_02}}
 {id=WavePlayer_7, properties {soundWaveAssetPtr=&/Force_Projectile_WaterSplash_03}}
 {id=WavePlayer_8, properties {soundWaveAssetPtr=&/Force_Projectile_WaterSplash_04}}
 {id=WavePlayer_9, properties {soundWaveAssetPtr=&/Force_Projectile_WaterSplash_05}}
]
properties {
 attenuationSettings=/Game/Sounds/Attenuation/Fort_ForcePowers_Projectile_WaterSplash_Att
 bOverrideConcurrency=true
 maxDistance=14000
 pitchMultiplier=0.9
 soundClassObject=&Fort_WeaponRanged_Fire_3P_PSM
 volumeMultiplier=0.9
 attenuationOverrides {falloffDistance=10000, attenuationShapeExtents {x=500}}
 concurrencyOverrides {maxCount=4, resolutionRule=StopOldest}
}
