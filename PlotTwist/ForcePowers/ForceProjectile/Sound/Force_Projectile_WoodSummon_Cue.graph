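# Sound cue for the wood summon sound of the Force Projectile power, layering wood impacts with delayed wood debris.
# A mixer combines an unmodulated impact layer with a quieter debris layer that is briefly delayed and passes through a default modulator.
# Each layer's quality switch picks randomly from six waves at the first two levels and three at the others.
# The cue raises pitch, reuses water-splash projectile attenuation, and caps concurrency by stopping the oldest.
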SoundCue /PlotTwist/ForcePowers/ForceProjectile/Sound/Force_Projectile_WoodSummon_Cue
root=Mixer_0
nodes [
 {id=Delay_0, children [QualityLevel_1], properties {delayMax=0.2, delayMin=0.2}}
 {id=Mixer_0, children [Modulator_2,Modulator_0], properties {inputVolume [1,0.7]}}
 {id=Modulator_0, children [Delay_0]}
 {id=Modulator_2, children [QualityLevel_0], properties {pitchMax=1, pitchMin=1, volumeMax=1, volumeMin=1}}
 {id=QualityLevel_0, children [Random_0,Random_0,Random_1,Random_1]}
 {id=QualityLevel_1, children [Random_2,Random_2,Random_3,Random_3]}
 {
  id=Random_0
  children [WavePlayer_1, WavePlayer_2, WavePlayer_3, WavePlayer_4, WavePlayer_5, WavePlayer_10]
  properties {weights [1,1,1,1,1,1]}
 }
 {id=Random_1, children [WavePlayer_4, WavePlayer_5, WavePlayer_10], properties {weights [1,1,1]}}
 {
  id=Random_2
  children [WavePlayer_0, WavePlayer_6, WavePlayer_7, WavePlayer_8, WavePlayer_9, WavePlayer_11]
  properties {weights [1,1,1,1,1,1]}
 }
 {id=Random_3, children [WavePlayer_8, WavePlayer_9, WavePlayer_11], properties {weights [1,1,1]}}
 {id=WavePlayer_0, properties {soundWaveAssetPtr=&/Force_Projectile_WoodDebris_06}}
 {id=WavePlayer_1, properties {soundWaveAssetPtr=&/Force_Projectile_Woodmpact_01}}
 {id=WavePlayer_10, properties {soundWaveAssetPtr=&/Force_Projectile_Woodmpact_06}}
 {id=WavePlayer_11, properties {soundWaveAssetPtr=&/Force_Projectile_WoodDebris_01}}
 {id=WavePlayer_2, properties {soundWaveAssetPtr=&/Force_Projectile_Woodmpact_02}}
 {id=WavePlayer_3, properties {soundWaveAssetPtr=&/Force_Projectile_Woodmpact_03}}
 {id=WavePlayer_4, properties {soundWaveAssetPtr=&/Force_Projectile_Woodmpact_04}}
 {id=WavePlayer_5, properties {soundWaveAssetPtr=&/Force_Projectile_Woodmpact_05}}
 {id=WavePlayer_6, properties {soundWaveAssetPtr=&/Force_Projectile_WoodDebris_02}}
 {id=WavePlayer_7, properties {soundWaveAssetPtr=&/Force_Projectile_WoodDebris_03}}
 {id=WavePlayer_8, properties {soundWaveAssetPtr=&/Force_Projectile_WoodDebris_04}}
 {id=WavePlayer_9, properties {soundWaveAssetPtr=&/Force_Projectile_WoodDebris_05}}
]
properties {
 attenuationSettings=/Game/Sounds/Attenuation/Fort_ForcePowers_Projectile_WaterSplash_Att
 bOverrideConcurrency=true
 maxDistance=14000
 pitchMultiplier=1.3
 soundClassObject=&Fort_WeaponRanged_Fire_3P_PSM
 volumeMultiplier=0.7
 attenuationOverrides {falloffDistance=10000, attenuationShapeExtents {x=500}}
 concurrencyOverrides {maxCount=4, resolutionRule=StopOldest}
}
